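# Carbon fiber skin material for the Arrow vehicle body's Livery01, with Default Lit and Clear Coat shading.
# A trim mask blends a color parameter with livery base color, multiplied by a clamped, tiled weave texture borrowed from weapon wraps.
# The weave normal's green channel drives roughness, and its flattened normal becomes the clear-coat bottom normal.
# A BlendShadingModel subgraph picks Clear Coat whenever coat amount is above zero, then feeds MF_Vehicle_FNPostFX.
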Material /VehicleCosmetics/Bodies/Arrow/Skins/Livery01/M_Skin_Arrow_Livery01_CarbonFiber
graph {
 properties {
  bUseMaterialAttributes=true
  bUsedWithMeshDeformer=true
  bUsedWithSkeletalMesh=true
  pixelDepthOffsetMode=PDOM_Legacy
  shadingModel=MSM_FromMaterialExpression
  usedShadingModels=Default Lit | Clear Coat
 }
 nodes [
  n1 = VectorParameter() {group=CarbonFiber, parameterName=CarbonFiberColor}
  n2 = TextureSample() {texture=&/Skin_Arrow_Livery01_TRIM_BC}
  n3 = TextureSample() {texture=&/Skin_Arrow_Livery01_TRIM_RGB}
  n4 = Multiply(A=TextureCoordinate() {uTiling=16, vTiling=16}, B=ScalarParameter() {defaultValue=1, group=CarbonFiber, parameterName=CarbonFiber__Tiling})
  n5 = TextureSample(UVs=@n4) {texture=/Game/Weapons/WeaponSkins/Wraps/Textures/T_Wrap_CarbonFiber_D}
  n6 = TextureSample(UVs=@n4) {samplerType=SAMPLERTYPE_Normal, texture=/Game/Weapons/WeaponSkins/Wraps/Textures/T_Wrap_CarbonFiber_N}
  n7 = ScalarParameter() {defaultValue=1, group=CarbonFiber, parameterName=CarbonFiber_ClearCoat}
  n8 = SetMaterialAttributes!(
   MaterialAttributes
   Base Color=Multiply(A=LinearInterpolate(A=@n1.RGB, B=@n2.RGB, Alpha=@n3.R), B=Clamp(None=@n5.RGB) {minDefault=0.85})
   Metallic=Constant() {r=0.1}
   Specular
   Roughness=Add(A=@n6.G, B=Constant() {r=0.2})
   Anisotropy
   Emissive Color
   Opacity
   Opacity Mask
   Normal
   Tangent
   World Position Offset
   Displacement
   Subsurface Color
   Clear Coat or Custom Data 0=@n7
   Clear Coat Roughness or Custom Data 1=ScalarParameter() {group=CarbonFiber, parameterName=CarbonFiber_ClearCoatRoughness}
   ClearCoatBottomNormal=MaterialFunctionCall(Normal=@n6.RGB, Flatness=ScalarParameter() {defaultValue=0.25, group=CarbonFiber, parameterName=CarbonFiber_Flatten}) {materialFunction=&FlattenNormal}.Result
   Shading Model=ShadingModel() {shadingModel=MSM_ClearCoat}
  ) {attributeSetTypes [
    #1
    #2
    #3
    #4
    #5
    #6
    #7
    #8
    #9
    #10
    #11
    #12
    #13
    #14
    #15
    #16
    #17
   ]}
  MaterialExpressionReroute_1 = Reroute(None=@n8)
  n9 = GetMaterialAttributes(None=@MaterialExpressionReroute_1."In (MA)") {attributeGetTypes [#9, #16]}
  MaterialExpressionReroute_2 = Reroute(None=@n7)
  n10 = BlendMaterialAttributes(
   A=SetMaterialAttributes!(
    MaterialAttributes=@n9.MaterialAttributes
    Shading Model=ShadingModel()
    Normal=MaterialFunctionCall(BaseNormal=@n9.Normal, AdditionalNormal=@n9.Normal) {materialFunction=&BlendAngleCorrectedNormals}.Result
   ) {attributeSetTypes [#17, #9]}
   B=@MaterialExpressionReroute_1."In (MA)"
   Alpha=Saturate(None=Ceil(None=@MaterialExpressionReroute_2.CCThickness))
  )
  MaterialExpressionReroute_3 = Reroute(None=@n10)
  n11 = GetMaterialAttributes(None=@MaterialExpressionReroute_3.Output) {attributeGetTypes [#16]}
  ClearCoatNormalCustomOutput!(None=@n11.ClearCoatBottomNormal)
  Composite!("In (MA)"=@n8, CCThickness=@n7) {subgraphName=BlendShadingModel}
  MaterialFunctionCall!(In=@n11.MaterialAttributes) {materialFunction=/VehicleCosmetics/Materials/MaterialFunctions/MF_Vehicle_FNPostFX}
  PinBase() {reroutePins [
    "In (MA)" {expression=&self:MaterialExpressionReroute_1}
    CCThickness {expression=&self:MaterialExpressionReroute_2}
   ]}
  PinBase!(Output=@n10) {reroutePins [Output {expression=&self:MaterialExpressionReroute_3}]}
 ]
}
